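; Audio gameplay function library that picks sound variants by whether the involved actors are locally viewed.
; Damage events return the receiver, instigator or observer sound in that priority, and player events return a first- or third-person sound.
; Both rely on IsActorLocal, which resolves player states to pawns, weapons to owners and vehicles to a locally viewed passenger, then checks whether that player pawn is locally viewed.
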(fn GetAudioForDamageEvent
  (params (Receiver :type Actor%20Object%20Reference) (Instigator :type Actor%20Object%20Reference) (ReceiverSound :type Sound%20Base%20Object%20Reference) (InstigatorSound :type Sound%20Base%20Object%20Reference) (ObserverSound :type Sound%20Base%20Object%20Reference))
  (if (Internal|IsActorLocal (Variables|GetReceiver))
    (return (Variables|GetReceiverSound))
    (elif (Internal|IsActorLocal (Variables|GetInstigator))
      (return (Variables|GetInstigatorSound))
      (else
        (return (Variables|GetObserverSound))))))

(fn GetAudioForPlayerEvent
  (params (Target :type Actor%20Object%20Reference) (1P%20Sound :type Sound%20Base%20Object%20Reference) (3P%20Sound :type Sound%20Base%20Object%20Reference))
  (if (Internal|IsActorLocal Target)
    (return 1P%20Sound)
    (else
      (return 3P%20Sound))))

(fn IsActorLocal
  (params (Actor :type Actor%20Object%20Reference))
  (bind _output_get (LocalVariables|SetFinalActor Actor))
  (if (Utilities|IsValid _output_get :!class /Script/Engine.KismetSystemLibrary)
    (Utilities|FlowControl|Sequence
      (:then_0
        (bind _asfort_player_state (Utilities|Casting|CastToFortPlayerState (LocalVariables|GetFinalActor))
          (:then
            (LocalVariables|SetFinalActor (PlayerState|GetPawn _asfort_player_state)))
          (:CastFailed
            (bind _asfort_weapon (Utilities|Casting|CastToFortWeapon (LocalVariables|GetFinalActor))
              (:then
                (LocalVariables|SetFinalActor (Actor|GetOwner _asfort_weapon)))
              (:CastFailed
                (bind _asfort_athena_vehicle (Utilities|Casting|CastToFortAthenaVehicle (LocalVariables|GetFinalActor)))
                (for _array_element (Vehicle|GetAllPassengers _asfort_athena_vehicle)
                  (if (Pawn|IsLocallyViewed _array_element)
                    (LocalVariables|SetFinalActor _array_element))))))))
      (:then_1
        (bind _asfort_player_pawn (Utilities|Casting|CastToFortPlayerPawn (LocalVariables|GetFinalActor))
          (:then
            (return (Pawn|IsLocallyViewed _asfort_player_pawn)))
          (:CastFailed
            (return false)))))
    (else
      (return false))))
