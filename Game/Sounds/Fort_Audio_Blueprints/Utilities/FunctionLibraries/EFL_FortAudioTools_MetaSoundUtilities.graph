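# Editor-only Fort Audio Tools utility library for inspecting and authoring MetaSound Sources and presets.
# Builder, editor and asset subsystems drive queries of inputs, preset parents and one-shot status, create presets from literal maps, and reassign duplicate class names while replacing references.
# Migrate Source Properties copies sound settings between assets through direct setters or editor-property access by name, logging per-property results that helpers filter by exception names and print.
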EditorUtilityBlueprint /Game/Sounds/Fort_Audio_Blueprints/Utilities/FunctionLibraries/EFL_FortAudioTools_MetaSoundUtilities
graphs [
 Add Message to Result {nodes [
   n1 = Add Message to Result!() -> (then, Result:"Fort Audio Tools Result Structure (by ref)", Message:"Text (by ref)") {is=title K2Node_FunctionEntry}
   n2 = Break Fort Audio Tools Result!(FortAudioTools_Result:"Fort Audio Tools Result Structure (by ref)"=@n1) -> (
    ResultType_14_#1:EFortAudioToolsResultType Enum
    Name_13_#2:Name
    Messages_12_#3:Array of Texts
   ) {is=title K2Node_BreakStruct}
   > n3 = Set Messages(Messages:Array of Texts=@n2.Messages_12_#3) -> (Output_Get:Array of Texts)
   > Add(TargetArray:Array of Texts=@n3, NewItem:"Text (by ref)"=@n1.Message) {is=title K2Node_CallArrayFunction}
   > Return Node!(
    execute
    OutResult:Fort Audio Tools Result Structure=Make Fort Audio Tools Result!(
     ResultType_14_#1:EFortAudioToolsResultType Enum=@n2
     Name_13_#2:Name=@n2.Name_13_#2
     Messages_12_#3:Array of Texts=@n3
    ) -> (FortAudioTools_Result:Fort Audio Tools Result Structure) {is=title K2Node_MakeStruct}
   )
  ]}
 Check Results for Exceptions {nodes [
   n1 = Check Results for Exceptions!() -> (then, Results:Array of Fort Audio Tools Result Structures, Exceptions:Array of Names) {is=title K2Node_FunctionEntry}
   > n2 = For Each Loop(Array:Array of Fort Audio Tools Result Structures=@n1) -> (Array Element:Fort Audio Tools Result Structure)
   Return Node!(
    execute=@n2.Completed
    Success:bool=Get Local Success() -> (Local Success:bool)
    Filtered Results:Array of Fort Audio Tools Result Structures=Get Local Results() -> (Local Results:Array of Fort Audio Tools Result Structures)
   )
   > n3 = Set Result(Result:Fort Audio Tools Result Structure=@n2) -> (Output_Get:Fort Audio Tools Result Structure)
   n4 = Break Fort Audio Tools Result(FortAudioTools_Result=@n3)
   > n5 = Switch on EFortAudioToolsResultType!(execute, Selection:EFortAudioToolsResultType Enum=@n4) -> (NewEnumerator1:Exec, NewEnumerator0:Exec) {is=title K2Node_SwitchEnum}
   n6 = Get Exceptions() -> (Exceptions:Array of Names)
   Branch(execute=@n5.NewEnumerator0, Condition=NOT Boolean(A=Contains Item(TargetArray=@n6, ItemToFind=@n4.Name_13_#2)))
   > n7 = Set Local Success(Local Success:bool=false) -> (Output_Get:bool)
   Add(
    execute=[@n7.then, @n5.NewEnumerator1]
    TargetArray:Array of Fort Audio Tools Result Structures=Get Local Results()
    NewItem:"Fort Audio Tools Result Structure (by ref)"=Get Result() -> (Result:Fort Audio Tools Result Structure)
   )
  ]}
 Create MetaSound Preset with Literals {nodes [
   n1 = Create MetaSound Preset with Literals!() -> (
    then
    Parent MetaSound Source:Meta Sound Source Object Reference
    MetaSound Input Literals:Map of Names to Meta Sound Literal Structures
    Package Path:String
    Asset Name:String
    Suffix:String
   ) {is=title K2Node_FunctionEntry}
   > Is Valid(InputObject=@n1)
   > n2 = CreateSourcePresetBuilder!(
    execute
    self:Meta Sound Builder Subsystem Object Reference=Get MetaSoundBuilderSubsystem() -> (ReturnValue:Meta Sound Builder Subsystem Object Reference) {is=title K2Node_GetEngineSubsystem}
    BuilderName:Name=PresetMigrator
    ReferencedSourceClass:"Meta Sound Document Interface Interface (by ref)"=@n1
   ) -> (Succeeded:Exec, Failed:Exec, ReturnValue:Meta Sound Source Builder Object Reference)
   > Set Preset Builder(Preset Builder:Meta Sound Source Builder Object Reference=@n2) -> (Output_Get:Meta Sound Source Builder Object Reference)
   > n3 = Keys!(
    execute
    TargetMap:Map of Names to Meta Sound Literal Structures=Get MetaSound Input Literals() -> (MetaSound Input Literals:Map of Names to Meta Sound Literal Structures)
   ) -> (then, Keys:Array of Names)
   > n4 = For Each Loop(Array:Array of Names=@n3) -> (Array Element:Name)
   n5 = GetAssetTools(execute=@n4.Completed)
   > n6 = CreateUniqueAssetName!(
    execute
    self:Asset Tools Interface=@n5
    InBasePackageName:String=Append!(A=@n1.Package Path, B=/, C:String=@n1.Asset Name, D:String=@n1.Suffix) -> (ReturnValue)
    InSuffix:String
   ) -> (then, OutPackageName:String, OutAssetName:String)
   > n7 = BuildToAsset!(
    execute
    self:Meta Sound Editor Subsystem Object Reference=Get MetaSoundEditorSubsystem() -> (ReturnValue:Meta Sound Editor Subsystem Object Reference) {is=title K2Node_GetEditorSubsystem}
    InBuilder:Meta Sound Builder Base Object Reference=Get Preset Builder() -> (Preset Builder:Meta Sound Source Builder Object Reference)
    Author:String=GetPlatformUserName!() -> (ReturnValue:String)
    AssetName:String=@n6.OutAssetName
    PackagePath:String=@n1.Package Path
    TemplateSoundWave:Sound Wave Object Reference
   ) -> (Succeeded:Exec, Failed:Exec, ReturnValue:Meta Sound Document Interface Interface)
   > n8 = Cast To MetaSoundSource!(execute, Object:Meta Sound Document Interface Interface=@n7) -> (then, CastFailed, AsMetaSound Source:Meta Sound Source Object Reference) {is=title K2Node_DynamicCast}
   > Set QualitySetting!(
    execute
    QualitySetting:Name=Get QualitySetting!(self:Meta Sound Source Object Reference=Get Parent MetaSound Source() -> (Parent MetaSound Source:Meta Sound Source Object Reference)) -> (QualitySetting:Name=None)
    self:Meta Sound Source Object Reference=@n8
   ) -> (then, Output_Get:Name=None)
   > Return Node!(execute, MetaSound Source Preset:Meta Sound Source Object Reference=@n8)
   n9 = Set Input Name(execute=@n4.LoopBody, Input Name:Name=@n4) -> (Output_Get:Name=None)
   n10 = Find(TargetMap:Map of Names to Meta Sound Literal Structures=Get MetaSound Input Literals(), Key:"Name (by ref)"=@n9) -> (Value:Meta Sound Literal Structure)
   > Branch(Condition=@n10.ReturnValue)
   > SetGraphInputDefault!(
    execute
    self:Meta Sound Builder Base Object Reference=Get Preset Builder()
    InputName:Name=@n9
    Literal:"Meta Sound Literal Structure (by ref)"=@n10
    PageName:Name
   ) -> (Succeeded:Exec, Failed:Exec)
  ]}
 Fix Up Duplicate References {nodes [
   n1 = Fix Up Duplicate References!() -> (then, MetaSound Asset:Object Reference, Reference Directories:Array of Meta Sound Asset Directory Structures) {is=title K2Node_FunctionEntry}
   > n2 = Cast To MetaSoundDocumentInterface!(execute, Object=@n1) -> (then, CastFailed, AsMetaSound Document Interface:Meta Sound Document Interface Interface) {is=title K2Node_DynamicCast}
   > n3 = Set Builder(Builder:Meta Sound Builder Base Object Reference=Find Or Begin Building MetaSound Asset!(self:Meta Sound Editor Subsystem Object Reference=Get MetaSoundEditorSubsystem(), MetaSound:Meta Sound Document Interface Interface=@n2) -> (ReturnValue:Meta Sound Builder Base Object Reference)) -> (Output_Get:Meta Sound Builder Base Object Reference)
   > Set Old Class Name(Old Class Name:Meta Sound Class Name Structure=Get MetaSound Class Name!(self:Meta Sound Builder Base Object Reference=@n3) -> (ReturnValue:Meta Sound Class Name Structure) {is=title K2Node_CallFunction}) -> (Output_Get:Meta Sound Class Name Structure)
   > n4 = ReassignClassName!(
    execute
    self:Meta Sound Asset Subsystem Object Reference=Get MetaSoundAssetSubsystem() -> (ReturnValue:Meta Sound Asset Subsystem Object Reference) {is=title K2Node_GetEngineSubsystem}
    DocInterface:Meta Sound Document Interface Interface=@n2
   ) -> (then, ReturnValue:bool)
   > Branch(Condition=@n4)
   > ReplaceReferencesInDirectory!(
    execute
    self:Meta Sound Asset Subsystem Object Reference=Get MetaSoundAssetSubsystem()
    InDirectories:Array of Meta Sound Asset Directory Structures=Get Reference Directories() -> (Reference Directories:Array of Meta Sound Asset Directory Structures)
    OldClassName:"Meta Sound Class Name Structure (by ref)"=Get Old Class Name() -> (Old Class Name:Meta Sound Class Name Structure)
    NewClassName:"Meta Sound Class Name Structure (by ref)"=Get MetaSound Class Name(self=Get Builder() -> (Builder:Meta Sound Builder Base Object Reference))
    OldVersion:Metasound Frontend Version Number Structure=(Major=1,Minor=0)
    NewVersion:Metasound Frontend Version Number Structure=(Major=1,Minor=0)
   ) -> (then, ReturnValue:bool)
  ]}
 Get All MetaSound Input Handles {nodes [
   n1 = Get All MetaSound Input Handles!() -> (then, MetaSound Source:Meta Sound Source Object Reference) {is=title K2Node_FunctionEntry}
   > Is Valid(InputObject=@n1)
   > Set Builder(Builder=Find Builder By MetaSound!(
     self:Meta Sound Builder Subsystem Object Reference=Get MetaSoundBuilderSubsystem()
     InMetaSound:Meta Sound Document Interface Interface=Get MetaSound Source() -> (MetaSound Source:Meta Sound Source Object Reference)
    ) -> (ReturnValue:Meta Sound Builder Base Object Reference))
   > n2 = Add MetaSound Node From Asset Class!(
    execute
    self:Meta Sound Builder Base Object Reference=Get Builder()
    NodeClass:"Meta Sound Document Interface Interface (by ref)"=Get MetaSound Source()
   ) -> (Succeeded:Exec, Failed:Exec, ReturnValue:Meta Sound Node Handle Structure)
   > n3 = FindNodeInputs!(execute, self:Meta Sound Builder Base Object Reference=Get Builder(), NodeHandle:"Meta Sound Node Handle Structure (by ref)"=@n2) -> (Succeeded:Exec, Failed:Exec, ReturnValue:Array of Meta Sound Node Input Handle Structures)
   > Return Node!(
    execute
    Source Builder:Meta Sound Builder Base Object Reference=Get Builder()
    Input Handles:Array of Meta Sound Node Input Handle Structures=@n3
   )
  ]}
 Get All MetaSound Source Input Literals {nodes [
   Get All MetaSound Source Input Literals!() -> (then, MetaSound Source:Meta Sound Source Object Reference) {is=title K2Node_FunctionEntry}
   > n1 = Set Builder(Builder=Find Builder By MetaSound(self=Get MetaSoundBuilderSubsystem(), InMetaSound=Get MetaSound Source()))
   > n2 = For Each Loop(Array=GetGraphInputNames!(self:Meta Sound Builder Base Object Reference=@n1) -> (ReturnValue:Array of Names))
   Return Node!(
    execute=@n2.Completed
    Input Default Literals:Map of Names to Meta Sound Literal Structures=Get Input Literals() -> (Input Literals:Map of Names to Meta Sound Literal Structures)
   )
   > Add!(
    execute
    TargetMap:Map of Names to Meta Sound Literal Structures=Get Input Literals()
    Key:"Name (by ref)"=@n2
    Value:"Meta Sound Literal Structure (by ref)"=GetGraphInputDefault!(self:Meta Sound Builder Base Object Reference=Get Builder(), InputName:Name=@n2, PageName:Name) -> (ReturnValue:Meta Sound Literal Structure)
   ) -> (then) {is=title K2Node_CallFunction}
  ]}
 Get All MetaSound Source Input Types {nodes [
   Get All MetaSound Source Input Types!() -> (then, MetaSound Source:Meta Sound Source Object Reference) {is=title K2Node_FunctionEntry}
   > n1 = Set Builder(Builder=Find Builder By MetaSound(self=Get MetaSoundBuilderSubsystem(), InMetaSound=Get MetaSound Source()))
   > n2 = For Each Loop(Array=GetGraphInputNames(self=@n1))
   Return Node!(execute=@n2.Completed, Input Types:Map of Names to Names=Get MetaSound Input Types() -> (MetaSound Input Types:Map of Names to Names))
   > n3 = FindGraphInputNode!(execute, self:Meta Sound Builder Base Object Reference=Get Builder(), InputName:Name=@n2) -> (
    Succeeded:Exec
    Failed:Exec
    OutDataType:Name=None
    NodeOutputHandle:Meta Sound Node Output Handle Structure
    ReturnValue:Meta Sound Node Handle Structure
   )
   > Add(TargetMap:Map of Names to Names=Get MetaSound Input Types(), Key=@n2, Value:"Name (by ref)"=@n3)
  ]}
 Get Preset Parent {nodes [
   n1 = Get Preset Parent!() -> (then, MetaSound Source Preset:Meta Sound Source Object Reference) {is=title K2Node_FunctionEntry}
   > Is Valid(InputObject=@n1)
   n2 = Get MetaSoundEditorSubsystem()
   > n3 = Cast To MetaSoundSource(Object:Object Reference=Get Preset Parent!(self:Meta Sound Builder Base Object Reference=Find Or Begin Building MetaSound Asset(self=@n2, MetaSound=@n1)) -> (ReturnValue:Object Reference) {is=title K2Node_CallFunction})
   > Return Node!(execute, Parent MetaSound Source:Meta Sound Source Object Reference=@n3)
  ]}
 Get Selected MetaSound Source Presets {nodes [
   Get Selected MetaSound Source Presets() {is=title K2Node_FunctionEntry}
   > n1 = GetSelectedAssets()
   > n2 = For Each Loop(Array:Array of Object References=@n1) -> (Array Element:Object Reference)
   Return Node!(
    execute=@n2.Completed
    Selected MetaSound Source Presets:Array of Meta Sound Source Object References=Get MetaSound Source Presets() -> (MetaSound Source Presets:Array of Meta Sound Source Object References)
   )
   > n3 = Cast To MetaSoundSource(Object=@n2)
   > Branch(Condition=Is Preset!(MetaSound Source:Meta Sound Source Object Reference=@n3, __WorldContext:Object Reference) -> (ReturnValue:bool))
   > Add(
    TargetArray:Array of Meta Sound Source Object References=Get MetaSound Source Presets()
    NewItem:"Meta Sound Source Object Reference (by ref)"=@n3
   ) {is=title K2Node_CallArrayFunction}
  ]}
 Is Oneshot {nodes [
   n1 = Is Oneshot!() -> (then, MetaSound Source:Meta Sound Source Object Reference) {is=title K2Node_FunctionEntry}
   > Is Valid(InputObject=@n1)
   n2 = Get MetaSoundEditorSubsystem()
   > Return Node!(
    execute
    ReturnValue:bool=InterfaceIsDeclared!(
     self:Meta Sound Builder Base Object Reference=Find Or Begin Building MetaSound Asset(self=@n2, MetaSound=@n1)
     InterfaceName:Name=UE.Source.OneShot
    ) -> (ReturnValue:bool)
   )
  ]}
 Is Preset {nodes [
   n1 = Is Preset!() -> (then, MetaSound Source:Meta Sound Source Object Reference) {is=title K2Node_FunctionEntry}
   > Is Valid(InputObject=@n1)
   n2 = Get MetaSoundEditorSubsystem()
   n3 = Get MetaSound Source()
   > Return Node(ReturnValue=IsPreset!(self:Meta Sound Builder Base Object Reference=Find Or Begin Building MetaSound Asset(self=@n2, MetaSound=@n3)) -> (ReturnValue:bool))
  ]}
 Migrate Source Properties {nodes [
   Migrate Source Properties!() -> (then, Asset to Migrate From:Sound Base Object Reference, Asset to Migrate To:Sound Base Object Reference) {is=title K2Node_FunctionEntry}
   > Get!(execute:Exec) -> (then:Exec, else:Exec, Asset to Migrate From:Sound Base Object Reference)
   > Get!(execute) -> (then, else, Asset to Migrate To:Sound Base Object Reference)
   > n1 = Sequence()
   > Set Current Property Name(Current Property Name:Name=OverrideConcurrency) -> (Output_Get:Name=None)
   > Set Current Result(Current Result:Fort Audio Tools Result Structure=Make Fort Audio Tools Result(
     ResultType_14_#1=NewEnumerator1
     Name_13_#2=Get Message Name() -> (Message Name:Name=None)
    )) -> (Output_Get:Fort Audio Tools Result Structure)
   > n2 = Set bOverrideConcurrency!(
    execute
    bOverrideConcurrency:bool=Get bOverrideConcurrency!(self:Sound Base Object Reference=Get Asset to Migrate From() -> (Asset to Migrate From:Sound Base Object Reference)) -> (bOverrideConcurrency:bool)
    self:Sound Base Object Reference=Get Asset to Migrate To() -> (Asset to Migrate To:Sound Base Object Reference)
   ) -> (then, Output_Get:bool)
   n3 = Add Message to Result!(
    execute=[
     @n43.then
     @n54.then
     @n59.then
     @n64.then
     @n69.then
     @n30.then
     @n24.then
     @n22.then
     @n20.then
     @n18.then
     @n16.then
     @n14.then
     @n12.then
     @n10.then
     @n8.then
     @n6.then
     @n2.then
    ]
    Result:"Fort Audio Tools Result Structure (by ref)"=Get Current Result() -> (Current Result:Fort Audio Tools Result Structure)
    Message:"Text (by ref)"=Format Text!(
     Format=Get Success Message Format() -> (Success Message Format:Text)
     property:Name=Get Current Property Name() -> (Current Property Name:Name=None)
     asset:Sound Base Object Reference=Get Asset to Migrate To()
    ) -> (Result)
    __WorldContext:Object Reference
   ) -> (then, OutResult:Fort Audio Tools Result Structure) {is=title K2Node_CallFunction}
   > n4 = Set Current Result(Current Result=@n3)
   Add(
    execute=[@n33.then, @n4.then]
    TargetArray:Array of Fort Audio Tools Result Structures=Get Local Results()
    NewItem:"Fort Audio Tools Result Structure (by ref)"=Get Current Result()
   )
   n5 = Sequence(execute=@n1.then_1)
   > Set Current Property Name(Current Property Name=BypassVolumeScaleForPriority) -> (Output_Get=None)
   > Set Current Result(Current Result=Make Fort Audio Tools Result(
     ResultType_14_#1=NewEnumerator1
     Name_13_#2=Get Message Name() -> (Message Name=None)
    ))
   > n6 = Set bBypassVolumeScaleForPriority!(
    execute
    bBypassVolumeScaleForPriority:bool=Get bBypassVolumeScaleForPriority!(self:Sound Base Object Reference=Get Asset to Migrate From()) -> (bBypassVolumeScaleForPriority:bool)
    self:Sound Base Object Reference=Get Asset to Migrate To()
   ) -> (then, Output_Get:bool)
   n7 = Sequence(execute=@n5.then_1)
   > Set Current Property Name(Current Property Name=Priority) -> (Output_Get=None)
   > Set Current Result(Current Result=Make Fort Audio Tools Result(
     ResultType_14_#1=NewEnumerator1
     Name_13_#2=Get Message Name() -> (Message Name=None)
    ))
   > n8 = Set Priority!(
    execute
    Priority:float=Get Priority!(self:Sound Base Object Reference=Get Asset to Migrate From()) -> (Priority:float)
    self:Sound Base Object Reference=Get Asset to Migrate To()
   ) -> (then, Output_Get:float)
   n9 = Sequence(execute=@n7.then_1)
   > Set Current Property Name(Current Property Name=VirtualizationMode) -> (Output_Get=None)
   > Set Current Result(Current Result=Make Fort Audio Tools Result(
     ResultType_14_#1=NewEnumerator1
     Name_13_#2=Get Message Name() -> (Message Name=None)
    ))
   > n10 = Set VirtualizationMode!(
    execute
    VirtualizationMode:EVirtualizationMode Enum=Get VirtualizationMode!(self:Sound Base Object Reference=Get Asset to Migrate From()) -> (VirtualizationMode:EVirtualizationMode Enum=Disabled)
    self:Sound Base Object Reference=Get Asset to Migrate To()
   ) -> (then, Output_Get:EVirtualizationMode Enum=Disabled)
   n11 = Sequence(execute=@n9.then_1)
   > Set Current Property Name(Current Property Name=ConcurrencySet) -> (Output_Get=None)
   > Set Current Result(Current Result=Make Fort Audio Tools Result(
     ResultType_14_#1=NewEnumerator1
     Name_13_#2=Get Message Name() -> (Message Name=None)
    ))
   > n12 = Set ConcurrencySet!(
    execute
    ConcurrencySet:Set of Sound Concurrency Object References=Get ConcurrencySet!(self:Sound Base Object Reference=Get Asset to Migrate From()) -> (ConcurrencySet:Set of Sound Concurrency Object References)
    self:Sound Base Object Reference=Get Asset to Migrate To()
   ) -> (then, Output_Get:Set of Sound Concurrency Object References)
   n13 = Sequence(execute=@n11.then_1)
   > Set Current Property Name(Current Property Name=ConcurrencyOverrides) -> (Output_Get=None)
   > Set Current Result(Current Result=Make Fort Audio Tools Result(
     ResultType_14_#1=NewEnumerator1
     Name_13_#2=Get Message Name() -> (Message Name=None)
    ))
   > n14 = Set ConcurrencyOverrides!(
    execute
    ConcurrencyOverrides:Sound Concurrency Settings Structure=Get ConcurrencyOverrides!(self:Sound Base Object Reference=Get Asset to Migrate From()) -> (ConcurrencyOverrides:Sound Concurrency Settings Structure)
    self:Sound Base Object Reference=Get Asset to Migrate To()
   ) -> (then, Output_Get:Sound Concurrency Settings Structure)
   n15 = Sequence(execute=@n13.then_1)
   > Set Current Property Name(Current Property Name=SourceEffectChain) -> (Output_Get=None)
   > Set Current Result(Current Result=Make Fort Audio Tools Result(
     ResultType_14_#1=NewEnumerator1
     Name_13_#2=Get Message Name() -> (Message Name=None)
    ))
   > n16 = Set SourceEffectChain!(
    execute
    SourceEffectChain:Sound Effect Source Preset Chain Object Reference=Get SourceEffectChain!(self:Sound Base Object Reference=Get Asset to Migrate From()) -> (SourceEffectChain:Sound Effect Source Preset Chain Object Reference)
    self:Sound Base Object Reference=Get Asset to Migrate To()
   ) -> (then, Output_Get:Sound Effect Source Preset Chain Object Reference)
   n17 = Sequence(execute=@n15.then_1)
   > Set Current Property Name(Current Property Name=PostEffectBusSends) -> (Output_Get=None)
   > Set Current Result(Current Result=Make Fort Audio Tools Result(
     ResultType_14_#1=NewEnumerator1
     Name_13_#2=Get Message Name() -> (Message Name=None)
    ))
   > n18 = Set BusSends!(
    execute
    BusSends:Array of Sound Source Bus Send Info Structures=Get BusSends!(self:Sound Base Object Reference=Get Asset to Migrate From()) -> (BusSends:Array of Sound Source Bus Send Info Structures)
    self:Sound Base Object Reference=Get Asset to Migrate To()
   ) -> (then, Output_Get:Array of Sound Source Bus Send Info Structures)
   n19 = Sequence(execute=@n17.then_1)
   > Set Current Property Name(Current Property Name=PreEffectBusSends) -> (Output_Get=None)
   > Set Current Result(Current Result=Make Fort Audio Tools Result(
     ResultType_14_#1=NewEnumerator1
     Name_13_#2=Get Message Name() -> (Message Name=None)
    ))
   > n20 = Set PreEffectBusSends!(
    execute
    PreEffectBusSends:Array of Sound Source Bus Send Info Structures=Get PreEffectBusSends!(self:Sound Base Object Reference=Get Asset to Migrate From()) -> (PreEffectBusSends:Array of Sound Source Bus Send Info Structures)
    self:Sound Base Object Reference=Get Asset to Migrate To()
   ) -> (then, Output_Get:Array of Sound Source Bus Send Info Structures)
   n21 = Sequence(execute=@n19.then_1)
   > Set Current Property Name(Current Property Name=BaseSubmix) -> (Output_Get=None)
   > Set Current Result(Current Result=Make Fort Audio Tools Result(
     ResultType_14_#1=NewEnumerator1
     Name_13_#2=Get Message Name() -> (Message Name=None)
    ))
   > n22 = Set SoundSubmixObject!(
    execute
    SoundSubmixObject:Sound Submix Base Object Reference=Get SoundSubmixObject!(self:Sound Base Object Reference=Get Asset to Migrate From()) -> (SoundSubmixObject:Sound Submix Base Object Reference)
    self:Sound Base Object Reference=Get Asset to Migrate To()
   ) -> (then, Output_Get:Sound Submix Base Object Reference)
   n23 = Sequence(execute=@n21.then_1)
   > Set Current Property Name(Current Property Name=SubmixSends) -> (Output_Get=None)
   > Set Current Result(Current Result=Make Fort Audio Tools Result(
     ResultType_14_#1=NewEnumerator1
     Name_13_#2=Get Message Name() -> (Message Name=None)
    ))
   > n24 = Set SoundSubmixSends!(
    execute
    SoundSubmixSends:Array of Sound Submix Send Info Structures=Get SoundSubmixSends!(self:Sound Base Object Reference=Get Asset to Migrate From()) -> (SoundSubmixSends:Array of Sound Submix Send Info Structures)
    self:Sound Base Object Reference=Get Asset to Migrate To()
   ) -> (then, Output_Get:Array of Sound Submix Send Info Structures)
   n25 = Sequence(execute=@n23.then_1)
   > Set Current Property Name(Current Property Name=Volume) -> (Output_Get=None)
   > n26 = Set Current Result(Current Result=Make Fort Audio Tools Result(
     ResultType_14_#1=NewEnumerator1
     Name_13_#2=Get Message Name() -> (Message Name=None)
    ))
   n27 = GetEditorProperty!(
    execute:Exec=[@n26.then, @n37.then]
    Object:Object Reference=Get Asset to Migrate From()
    PropertyName:Name=Get Current Property Name() -> (Current Property Name=None)
   ) -> (then:Exec, PropertyValue:float, ReturnValue:bool)
   > n28 = Branch(Condition=@n27.ReturnValue)
   > n29 = SetEditorProperty(Object=Get Asset to Migrate To(), PropertyName=Get Current Property Name() -> (Current Property Name=None), PropertyValue:float=@n27)
   > n30 = Branch(Condition=@n29)
   n31 = Add Message to Result(
    execute=[@n43.else, @n54.else, @n59.else, @n64.else, @n69.else, @n30.else]
    Result=Get Current Result()
    Message=Format Text(
     Format=Get Fail Set Message Format() -> (Fail Set Message Format:Text)
     property=Get Current Property Name() -> (Current Property Name=None)
     asset=Get Asset to Migrate To()
    )
   )
   > n32 = Set Current Result(Current Result=@n31)
   n33 = Set Local Success(execute=[@n32.then, @n35.then], Local Success=false)
   n34 = Add Message to Result(
    execute=[@n52.else, @n57.else, @n62.else, @n67.else, @n28.else, @n41.else]
    Result=Get Current Result()
    Message=Format Text(
     Format=Get Fail Get Message Format() -> (Fail Get Message Format:Text)
     property=Get Current Property Name() -> (Current Property Name=None)
     asset=Get Asset to Migrate From()
    )
   )
   > n35 = Set Current Result(Current Result=@n34)
   n36 = Sequence(execute=@n25.then_1)
   > Set Current Property Name(Current Property Name=Pitch) -> (Output_Get=None)
   > n37 = Set Current Result(Current Result=Make Fort Audio Tools Result(
     ResultType_14_#1=NewEnumerator1
     Name_13_#2=Get Message Name() -> (Message Name=None)
    ))
   n38 = Sequence(execute=@n36.then_1)
   > Set Current Property Name(Current Property Name=bDebug) -> (Output_Get=None)
   > n39 = Set Current Result(Current Result=Make Fort Audio Tools Result(
     ResultType_14_#1=NewEnumerator1
     Name_13_#2=Get Message Name() -> (Message Name=None)
    ))
   n40 = GetEditorProperty(
    execute=[@n47.then, @n39.then, @n45.then, @n49.then]
    Object=Get Asset to Migrate From()
    PropertyName=Get Current Property Name() -> (Current Property Name=None)
   ) -> (PropertyValue:bool)
   > n41 = Branch(Condition=@n40.ReturnValue)
   > n42 = SetEditorProperty(Object=Get Asset to Migrate To(), PropertyName=Get Current Property Name() -> (Current Property Name=None), PropertyValue:bool=@n40)
   > n43 = Branch(Condition=@n42)
   n44 = Sequence(execute=@n38.then_1)
   > Set Current Property Name(Current Property Name=bEnableBusSends) -> (Output_Get=None)
   > n45 = Set Current Result(Current Result=Make Fort Audio Tools Result(
     ResultType_14_#1=NewEnumerator1
     Name_13_#2=Get Message Name() -> (Message Name=None)
    ))
   n46 = Sequence(execute=@n44.then_1)
   > Set Current Property Name(Current Property Name=bEnableBaseSubmix) -> (Output_Get=None)
   > n47 = Set Current Result(Current Result=Make Fort Audio Tools Result(
     ResultType_14_#1=NewEnumerator1
     Name_13_#2=Get Message Name() -> (Message Name=None)
    ))
   n48 = Sequence(execute=@n46.then_1)
   > Set Current Property Name(Current Property Name=bEnableSubmixSends) -> (Output_Get=None)
   > n49 = Set Current Result(Current Result=Make Fort Audio Tools Result(
     ResultType_14_#1=NewEnumerator1
     Name_13_#2=Get Message Name() -> (Message Name=None)
    ))
   n50 = Sequence(execute=@n48.then_1)
   > Set Current Property Name(Current Property Name=ModulationSettings) -> (Output_Get=None)
   > Set Current Result(Current Result=Make Fort Audio Tools Result(
     ResultType_14_#1=NewEnumerator1
     Name_13_#2=Get Message Name() -> (Message Name=None)
    ))
   > n51 = GetEditorProperty(Object=Get Asset to Migrate From(), PropertyName=Get Current Property Name() -> (Current Property Name=None)) -> (PropertyValue:"Sound Modulation Default Routing Settings Structure (by ref)")
   > n52 = Branch(Condition=@n51.ReturnValue)
   > n53 = SetEditorProperty(
    Object=Get Asset to Migrate To()
    PropertyName=Get Current Property Name() -> (Current Property Name=None)
    PropertyValue:"Sound Modulation Default Routing Settings Structure (by ref)"=@n51
   )
   > n54 = Branch(Condition=@n53)
   n55 = Sequence(execute=@n50.then_1)
   > Set Current Property Name(Current Property Name=SoundClassObject) -> (Output_Get=None)
   > Set Current Result(Current Result=Make Fort Audio Tools Result(
     ResultType_14_#1=NewEnumerator1
     Name_13_#2=Get Message Name() -> (Message Name=None)
    ))
   > n56 = GetEditorProperty(Object=Get Asset to Migrate From(), PropertyName=Get Current Property Name() -> (Current Property Name=None)) -> (PropertyValue:Sound Class Object Reference)
   > n57 = Branch(Condition=@n56.ReturnValue)
   > n58 = SetEditorProperty(
    Object=Get Asset to Migrate To()
    PropertyName=Get Current Property Name() -> (Current Property Name=None)
    PropertyValue:Sound Class Object Reference=@n56
   )
   > n59 = Branch(Condition=@n58)
   n60 = Sequence(execute=@n55.then_1)
   > Set Current Property Name(Current Property Name=AttenuationSettings) -> (Output_Get=None)
   > Set Current Result(Current Result=Make Fort Audio Tools Result(
     ResultType_14_#1=NewEnumerator1
     Name_13_#2=Get Message Name() -> (Message Name=None)
    ))
   > n61 = GetEditorProperty(Object=Get Asset to Migrate From(), PropertyName=Get Current Property Name() -> (Current Property Name=None)) -> (PropertyValue:Sound Attenuation Object Reference)
   > n62 = Branch(Condition=@n61.ReturnValue)
   > n63 = SetEditorProperty(
    Object=Get Asset to Migrate To()
    PropertyName=Get Current Property Name() -> (Current Property Name=None)
    PropertyValue:Sound Attenuation Object Reference=@n61
   )
   > n64 = Branch(Condition=@n63)
   n65 = Sequence(execute=@n60.then_1)
   > Set Current Property Name(Current Property Name=AudioPropertiesSheet) -> (Output_Get=None)
   > Set Current Result(Current Result=Make Fort Audio Tools Result(
     ResultType_14_#1=NewEnumerator1
     Name_13_#2=Get Message Name() -> (Message Name=None)
    ))
   > n66 = GetEditorProperty(Object=Get Asset to Migrate From(), PropertyName=Get Current Property Name() -> (Current Property Name=None))
   > n67 = Branch(Condition=@n66.ReturnValue)
   > n68 = SetEditorProperty(Object=Get Asset to Migrate To(), PropertyName=Get Current Property Name() -> (Current Property Name=None), PropertyValue=@n66)
   > n69 = Branch(Condition=@n68)
   Return Node!(execute=@n65.then_1, Success:bool=Get Local Success(), Results:Array of Fort Audio Tools Result Structures=Get Local Results())
  ]}
 Print Results {nodes [
   n1 = Print Results!() -> (then, Results:Array of Fort Audio Tools Result Structures, Exclude Successful Results:bool=true) {is=title K2Node_FunctionEntry}
   > n2 = For Each Loop(Array:Array of Fort Audio Tools Result Structures=@n1) -> (Array Element:Fort Audio Tools Result Structure)
   n3 = NOT Boolean(A=Get Exclude Successful Results() -> (Exclude Successful Results:bool))
   n4 = Break Fort Audio Tools Result(FortAudioTools_Result=@n2)
   n5 = "Equal (Enum)"(A:EFortAudioToolsResultType Enum=@n4, B:EFortAudioToolsResultType Enum=NewEnumerator1)
   > Branch(Condition=OR Boolean(
     A=AND Boolean(A=@n3, B=@n5)
     B="Not Equal (Enum)"!(A:EFortAudioToolsResultType Enum=@n4, B:EFortAudioToolsResultType Enum=NewEnumerator1) -> (ReturnValue:bool)
    ))
   > n6 = For Each Loop(Array:Array of Texts=@n4.Messages_12_#3) -> (Array Element:Text)
   > PrintText(
    InText=Format Text!(Format="LOCTEXT(\"{name}: {message}\")", name:Name=@n4.Name_13_#2, message:Text=@n6) -> (Result)
    bPrintToScreen=false
   )
  ]}
]
variables []
